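# Master character material for the Lone Wolf Super Level body, with switches for tattoo, Ronin tier and Motorcyclist variants.
# PreFX attributes are mask-blended with a black-albedo, emissive-driven look, optionally Motorcyclist-tinted, then fed to PostFX.
# The emissive combines a curve-atlas cel ramp from faked lighting, masked tints and edge glow, animated by triplanar noise in pre-skinned space.
# A radial spiral-mask branch ends unconnected.
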Material /BRCosmetics/Characters/Player/Male/Medium/Bodies/M_MED_Lone_Wolf/Materials/SuperLevel/MAT_LoneWolf_SL_Master
graph {
 properties {
  bUseMaterialAttributes=true
  bUsedWithMeshDeformer=true
  bUsedWithMorphTargets=true
  bUsedWithSkeletalMesh=true
  pixelDepthOffsetMode=PDOM_Legacy
  shadingModel=MSM_FromMaterialExpression
  usedShadingModels=Default Lit | Subsurface Profile
 }
 nodes [
  n1 = TextureSampleParameter2D() {
   parameterName=Diffuse
   samplerSource=SSM_Wrap_WorldGroupSettings
   texture=/BRCosmetics/Characters/Player/Male/Medium/Bodies/M_MED_Lone_Wolf/Textures/T_M_MED_Lone_Wolf_Body_D
  }
  n2 = NamedRerouteDeclaration(None=@n1.G) {name=Diffuse_G, nodeColor=<rgba 0 1 0.329 1>}
  n3 = NamedRerouteDeclaration(None=@n1.B) {name=Diffuse_B, nodeColor=<rgba 0.7813 0 1 1>}
  AppendVector(A=AppendVector(A=Constant(), B=@n2), B=@n3)
  Constant() {r=1}
  n4 = PixelNormalWS()
  Fresnel(ExponentIn=Constant() {r=0.2}, Normal=@n4)
  n5 = MaterialFunctionCall() {materialFunction=&BaseFN_CharacterFunction_PreFX}
  n6 = Constant()
  n7 = TextureObjectParameter() {parameterName=SkinFX_Mask, texture=/BRCosmetics/Characters/Player/Male/Medium/Bodies/M_MED_Lone_Wolf/Textures/T_M_MED_Lone_Wolf_Body_FX}
  n8 = StaticSwitchParameter(
   True=MaterialFunctionCall!(FXMask=@n7, MaterialAttributes=@n5.Result) {materialFunction=/Game/Weapons/WeaponSkins/MaterialFunction/MF_Tattoo}.Result
   False=@n5.Result
  ) {group=Skin FX, parameterName=UseTattoo}
  n9 = GetMaterialAttributes(None=@n8) {attributeGetTypes [#1]}
  n10 = NamedRerouteUsage()
  n11 = SetMaterialAttributes!(MaterialAttributes, Base Color=@n6, Normal=@n9.Normal, Emissive Color=@n10, Metallic=@n6, Specular=@n6, Roughness=Constant() {r=0.8}) {attributeSetTypes [
    #2
    #1
    #3
    #4
    #5
    #6
   ]}
  n12 = BlendMaterialAttributes(A=@n8, B=@n11, Alpha=NamedRerouteUsage())
  n13 = GetMaterialAttributes(None=@n12) {attributeGetTypes [#2, #3]}
  n14 = TextureSampleParameter2D() {parameterName=SkinFX_Mask, texture=&T_M_MED_Lone_Wolf_Body_FX}
  n15 = LinearInterpolate(
   A=@n13.Base Color
   B=VectorParameter() {parameterName=MotorcyclistDiffuseColor, defaultValue=<rgba 1 0.683 0 1>}.RGB
   Alpha=ChannelMaskParameter(None=@n14.RGB) {maskChannel=Green, parameterName=Motorcyclist_BasecolorTintChannel, defaultValue=<rgba 0 1 0 0>}.R
  )
  n16 = LinearInterpolate(
   A=@n13.Emissive Color
   B=VectorParameter() {parameterName=EmissiveColor, defaultValue=<rgba 1 0.683 0 1>}.RGB
   Alpha=ChannelMaskParameter(None=@n14.RGB) {maskChannel=Blue, parameterName=Motorcyclist_EmissiveTintChannel, defaultValue=<rgba 0 0 1 0>}.R
  )
  MaterialFunctionCall(
   Distance Blend=@n5.Distance Blend
   Fresnel=@n5.Distance Blend
   Pre Skinned Local Position=@n5.Distance Blend
   Material=StaticSwitchParameter(
    True=SetMaterialAttributes!(MaterialAttributes=@n13.MaterialAttributes, Base Color=@n15, Emissive Color=@n16) {attributeSetTypes [#2, #3]}
    False=@n12
   ) {parameterName=Motorcyclist}
  ) {materialFunction=&BaseFN_CharacterFunction_PostFX}
  MaterialFunctionCall!(Color=@n10) {materialFunction=/Game/Packages/Fortress_SharedMaterials/Anime_Master_Material/MF_TODColorCurve}
  n17 = Desaturation(None=@n1.RGB)
  n18 = NamedRerouteDeclaration(None=@n1.R) {name=Diffuse_R, nodeColor=<rgba 1 0 0.0933 1>}
  Max(A=Max(A=Max(A=@n17, B=@n18), B=@n2), B=@n3)
  n19 = ScalarParameter() {defaultValue=8, group=PanningEmissive, parameterName=PanningEmissiveTint1_Max}
  Multiply(A=@n19) {constB=0.2}
  Multiply(A=NamedRerouteUsage()) {constB=0.2}
  n20 = TextureSampleParameter2D(UVs=TextureCoordinate() {coordinateIndex=1}) {
   parameterName=Tattoo_Texture
   samplerSource=SSM_Wrap_WorldGroupSettings
   texture=/BRCosmetics/Characters/Player/Male/Medium/Bodies/M_MED_Lone_Wolf/Textures/T_M_MED_Lone_Wolf_Tattoo
  }
  n21 = Multiply(A=@n20.A) {constB=0.75}
  n22 = TextureSampleParameter2D() {parameterName=Non-skin Mask, samplerSource=SSM_Wrap_WorldGroupSettings, texture=&T_BlacksRGB}
  n23 = TextureSampleParameter2D() {
   parameterName=M
   samplerSource=SSM_Wrap_WorldGroupSettings
   samplerType=SAMPLERTYPE_Masks
   texture=/BRCosmetics/Characters/Player/Male/Medium/Bodies/M_MED_Lone_Wolf/Textures/T_M_MED_Lone_Wolf_Body_M
  }
  n24 = OneMinus(None=@n23.B)
  n25 = TextureSampleParameter2D() {
   parameterName=SpecularMasks
   samplerType=SAMPLERTYPE_Masks
   texture=/BRCosmetics/Characters/Player/Female/Medium/Bodies/F_MED_ExoSuit/Textures/T_F_MED_ExosuitHead_FaceAcc_S
  }
  n26 = Constant()
  n27 = LinearInterpolate(
   A=StaticSwitchParameter(True=ChannelMaskParameter(None=@n22.RGB) {parameterName=Non-skin Mask Channel}.R, False=Round(None=@n24)) {parameterName=Use Custon Non-skin Mask}
   B=If(A=@n25.G, B=Constant() {r=0.75}, "A > B"=Constant() {r=1}, "A == B"=@n26, "A < B"=@n26)
   Alpha=ScalarParameter() {parameterName="Use _S(GreenChannel) as Non-skin Mask"}
  )
  NamedRerouteDeclaration(None=StaticSwitchParameter(True=Add(A=@n21, B=@n27), False=@n27) {parameterName=LW_T2_Ronin}) {name=Non-skin Mask, nodeColor=<rgba 0.9922 0 1 1>}
  n28 = PixelNormalWS()
  n29 = Fresnel(ExponentIn=ScalarParameter() {defaultValue=0.75, group=SuperLevel, parameterName=TintMask_Fresnel_exp}, Normal=@n28)
  n30 = MaterialFunctionCall(In=Fresnel(ExponentIn=Constant() {r=2}, Normal=@n28), Contrast=Constant() {r=0.5}) {materialFunction=&CheapContrast}
  n31 = TextureSampleParameter2D() {
   parameterName=SL_Mask
   samplerType=SAMPLERTYPE_Masks
   texture=/BRCosmetics/Characters/Player/Male/Medium/Bodies/M_MED_Lone_Wolf/Textures/T_M_MED_Lone_Wolf_SL_Masks
  }
  n32 = Add(A=NamedRerouteUsage(), B=@n31.R)
  n33 = NamedRerouteUsage()
  n34 = LinearInterpolate(B=Add(A=@n32, B=Multiply(A=@n33, B=@n30.Result)), Alpha=@n32)
  n35 = StaticSwitchParameter(True=Add(A=@n31.R, B=NamedRerouteUsage()), False=@n31.R) {parameterName=LW_T2_Ronin}
  NamedRerouteDeclaration(None=LinearInterpolate(B=Add(A=@n29, B=@n30.Result), Alpha=StaticSwitchParameter(True=@n34, False=@n35) {parameterName=LW_T3_Ronin})) {name=Custom Tint Mask1, nodeColor=<rgba 1 0 0.6797 1>}
  n36 = MaterialFunctionCall(
   In=Fresnel(ExponentIn=ScalarParameter() {defaultValue=1.5, parameterName=Tint2EdgeDarkenFresnelExp}, Normal=@n4)
   Contrast=ScalarParameter() {defaultValue=1, group=SuperLevel, parameterName=Tint2EdgeContrast}
  ) {materialFunction=&CheapContrast}
  n37 = ScalarParameter() {defaultValue=1, parameterName=Tint2EdgeDarkenAmount}
  n38 = Saturate(None=Add(A=OneMinus(None=@n36.Result), B=OneMinus(None=@n37)) {constB=0.5})
  NamedRerouteDeclaration(None=LinearInterpolate(B=Multiply(A=@n29, B=@n38), Alpha=@n31.B)) {name=Custom Tint Mask2, nodeColor=<rgba 0 0.2122 1 1>}
  n39 = Subtract(
   A=CameraPositionWS()
   B=MaterialFunctionCall() {materialFunction=&ObjectPivotPoint}.Object Pivot Location
  )
  n40 = ComponentMask(None=Multiply(A=Normalize(VectorInput=@n39)) {constB=-1}) {b=true, g=true, r=true}
  n41 = Saturate(None=DotProduct(A=NamedRerouteUsage(), B=Normalize(VectorInput=@n40)))
  NamedRerouteDeclaration(None=@n41) {name=BackLitMask, nodeColor=<rgba 0 0.5547 1 1>}
  n42 = MaterialFunctionCall(Exponent=Constant() {r=1}) {materialFunction=&MF_Fresnel_FX}
  n43 = MaterialFunctionCall(Float3=PixelNormalWS()) {materialFunction=&BreakOutFloat3Components}
  n44 = LinearInterpolate(A=@n43.G, B=@n43.G, Alpha=NamedRerouteUsage())
  n45 = ConstantBiasScale(None=Sine(None=Add(A=@n42.Fresnel, B=@n44)) {period=0.5}) {bias=2}
  NamedRerouteDeclaration(None=Saturate(None=Multiply(A=@n42.Fresnel Inverted, B=@n45))) {name=??Mask, nodeColor=<rgba 0 0.0156 1 1>}
  n46 = Constant() {r=0.1}
  n47 = Constant() {r=0.5}
  n48 = PreSkinnedPosition()
  n49 = PreSkinnedLocalBounds()
  n50 = ComponentMask(None=Divide(A=VertexInterpolator(VS=@n48).PS, B=Subtract(A=@n49.Max, B=@n49.Max))) {b=true}
  n51 = OneMinus(None=@n50)
  NamedRerouteDeclaration(None=Multiply(A=SmoothStep(Min=@n46, Max=@n47, Value=@n51), B=@n20.A)) {name=LW_T2_TattooMask, nodeColor=<rgba 1 0.6797 0 1>}
  NamedRerouteDeclaration(None=@n20.A) {name=LW_T3_TattooMask, nodeColor=<rgba 0 1 0.5078 1>}
  NamedRerouteDeclaration(None=SmoothStep(
    Min=ScalarParameter() {parameterName=VerticalTransition_Min}
    Max=ScalarParameter() {defaultValue=0.2, parameterName=VerticalTransition_Max}
    Value=OneMinus(None=@n50)
   )) {name=VerticalTransition, nodeColor=<rgba 1 0.4688 0 1>}
  n52 = Multiply(A=Time()) {constB=0.35}
  n53 = ConstantBiasScale(None=Sine(None=Frac(None=@n52)))
  NamedRerouteDeclaration(None=@n53) {name=Pulse, nodeColor=<rgba 0.5 1 0 1>}
  NamedRerouteDeclaration() {name=test, nodeColor=<rgba 1 0 0.7272 1>}
  NamedRerouteDeclaration(None=@n31.G) {name=Edge Mask, nodeColor=<rgba 0 1 1 1>}
  n54 = CollectionParameter() {
   collection=&FortniteMaterialParameters
   parameterName=SunlightDirectionalVector
  }
  NamedRerouteDeclaration(None=Multiply(A=ComponentMask(None=@n54) {b=true, g=true, r=true}, B=Constant3Vector() {constant=<rgba 1 1 -1 0>})) {name=Sunlight, nodeColor=<rgba 1 0.4545 0 1>}
  NamedRerouteDeclaration(None=AppendVector(A=Constant(), B=ComponentMask(None=@n1.RGB) {b=true, g=true})) {name=Diffuse_0GB, nodeColor=<rgba 0.7109 1 0 1>}
  n55 = Multiply(A=VectorParameter() {parameterName=Camera Bias Vector, defaultValue=<rgba 0.06 0.1 0 1>}.RGB) {constB=-1}
  n56 = NamedRerouteUsage()
  n57 = Transform(None=NamedRerouteUsage())
  n58 = DotProduct(A=Normalize(VectorInput=Add(A=@n55, B=@n56)), B=@n57)
  n59 = DotProduct(A=NamedRerouteUsage(), B=@n57)
  n60 = Multiply(A=Time()) {constB=0.4}
  n61 = ConstantBiasScale(None=Sine(None=Frac(None=@n60)))
  n62 = Saturate(None=Add(A=LinearInterpolate(A=@n58, B=@n59, Alpha=@n61), B=ScalarParameter() {defaultValue=0.75, parameterName=SunShadow Strength}) {constB=0})
  n63 = Multiply(A=ReflectionVectorWS(), B=Constant3Vector() {constant=<rgba 1 1 -2 0>})
  n64 = CameraVectorWS()
  n65 = Multiply(A=@n62, B=OneMinus(None=DotProduct(A=@n63, B=@n64)))
  n66 = CurveAtlasRowParameter(CurveTime=Saturate(None=LinearInterpolate(A=@n65, B=@n62, Alpha=@n41))) {
   atlas=&/C_Atlas_S19_SL
   curve=&/C_LoneWolf_SL
   parameterName=Cel-Ramp
  }
  n67 = LinearInterpolate(
   A=VectorParameter() {group=SuperLevel, parameterName=BaseTint, defaultValue=<rgba 0.16 0.33 0.3958 1>}.RGB
   B=Power(Base=NamedRerouteUsage()) {constExponent=0.7}
   Alpha=ScalarParameter() {defaultValue=0.5, group=SuperLevel, parameterName=OriginalDiffusewithoutR_Strength}
  )
  n68 = NamedRerouteUsage()
  n69 = LinearInterpolate(
   A=Constant()
   B=VectorParameter() {group=SuperLevel, parameterName=BaseTint2_Strengh, defaultValue=<rgba 0.2495 0.355 0.3958 1>}.RGB
   Alpha=Saturate(None=Subtract(A=@n68) {constB=0.5})
  )
  n70 = LinearInterpolate(B=NamedRerouteUsage(), Alpha=ScalarParameter() {defaultValue=1, parameterName=BaseTint2Strength})
  n71 = VectorParameter() {group=SuperLevel, parameterName=EmissiveTint1_Primary, defaultValue=<rgba 0 0.3578 0.3646 1>}
  n72 = LinearInterpolate(
   A=VectorParameter() {group=SuperLevel, parameterName=EmissiveTint1_Secondary, defaultValue=<rgba 0 0.3578 0.3646 1>}.RGB
   B=@n71.RGB
   Alpha=NamedRerouteUsage()
  )
  n73 = ScalarParameter() {defaultValue=1.5, group=PanningEmissive, parameterName=PanningEmissiveTint1Min}
  n74 = TextureSampleParameter2D(UVs=TextureCoordinate() {uTiling=0.5, vTiling=0.5}) {
   parameterName=Panning Texture
   samplerSource=SSM_Wrap_WorldGroupSettings
   samplerType=SAMPLERTYPE_LinearGrayscale
   texture=/BRCosmetics/Athena/Effects/Particles/Skydiving/NS_Contrail_DarkAndLight/T_DarkAndLight_Noise
  }
  n75 = OneMinus(None=ScalarParameter() {defaultValue=0.1, group=PanningEmissive, parameterName=UVScaleMapIntensity})
  n76 = Multiply(A=Saturate(None=Add(A=@n74.R, B=@n75)), B=ScalarParameter() {defaultValue=0.5, parameterName=PanningEmissiveTiling})
  n77 = Divide(
   A=VertexInterpolator(VS=PreSkinnedPosition()).PS
   B=ScalarParameter() {defaultValue=40, group=PanningEmissive, parameterName=Triplanner_ConstantScaleDivisor}
  )
  n78 = Multiply(A=@n76, B=@n77) {constA=2}
  n79 = Frac(None=Multiply(A=Time(), B=VectorParameter() {group=PanningEmissive, parameterName=TriPlanner Speed, defaultValue=<rgba 0 -0.03 0.1 1>}.RGB))
  n80 = TextureSampleParameter2D(UVs=ComponentMask(None=Add(A=@n78, B=@n79)) {b=true, g=true}) {
   parameterName=Panning Texture
   samplerSource=SSM_Wrap_WorldGroupSettings
   samplerType=SAMPLERTYPE_LinearGrayscale
   texture=&T_DarkAndLight_Noise
  }
  n81 = Multiply(A=@n76, B=@n77) {constA=2}
  n82 = TextureSampleParameter2D(UVs=ComponentMask(None=Add(A=@n81, B=@n79)) {b=true, r=true}) {
   parameterName=Panning Texture
   samplerSource=SSM_Wrap_WorldGroupSettings
   samplerType=SAMPLERTYPE_LinearGrayscale
   texture=&T_DarkAndLight_Noise
  }
  n83 = VertexInterpolator(VS=PreSkinnedNormal())
  n84 = ScalarParameter() {defaultValue=1, group=PanningEmissive, parameterName=TriPlanner_AxisBlendStrength}
  n85 = MaterialFunctionCall(In=Abs(None=ComponentMask(None=@n83.PS) {g=true}), Contrast=@n84) {materialFunction=&CheapContrast}
  n86 = LinearInterpolate(A=@n80.RGB, B=@n82.RGB, Alpha=Saturate(None=ComponentMask(None=@n85.Result) {g=true}))
  n87 = Multiply(A=@n76, B=@n77) {constA=2}
  n88 = TextureSampleParameter2D(UVs=ComponentMask(None=Add(A=@n87, B=@n79)) {g=true, r=true}) {
   parameterName=Panning Texture
   samplerSource=SSM_Wrap_WorldGroupSettings
   samplerType=SAMPLERTYPE_LinearGrayscale
   texture=&T_DarkAndLight_Noise
  }
  n89 = Abs(None=ComponentMask(None=@n83.PS) {b=true})
  n90 = Saturate(None=ComponentMask(None=MaterialFunctionCall(In=@n89, Contrast=@n84) {materialFunction=&CheapContrast}.Result) {b=true})
  n91 = ComponentMask(None=LinearInterpolate(
    A=LinearInterpolate(A=@n86, B=@n88.RGB, Alpha=@n90)
    B=LinearInterpolate(A=@n82.RGB, B=@n88.RGB)
    Alpha=ScalarParameter() {group=PanningEmissive, parameterName=Triplanner_Z axis Only}
   )) {r=true}
  n92 = NamedRerouteUsage()
  n93 = ScalarParameter() {defaultValue=4, group=PanningEmissive, parameterName=PanningEmissive_Fresnel Exponent}
  n94 = Multiply(A=@n93) {constB=0.2}
  n95 = Saturate(None=Fresnel(ExponentIn=LinearInterpolate(A=@n93, B=@n94, Alpha=@n53) {constB=4}) {baseReflectFraction=0})
  n96 = LinearInterpolate(
   A=OneMinus(None=@n95)
   B=@n95
   Alpha=ScalarParameter() {defaultValue=0.75, group=PanningEmissive, parameterName=Panning Emissive_Invert Fresnel}
  )
  n97 = LinearInterpolate(
   A=Multiply(A=@n66, B=LinearInterpolate(A=@n67, B=@n69, Alpha=@n70) {constB=0})
   B=Multiply(A=StaticSwitchParameter(True=@n72, False=@n71.RGB) {parameterName=T3}, B=LinearInterpolate(A=@n73, B=@n19, Alpha=@n91))
   Alpha=Saturate(None=Multiply(A=@n92, B=@n96))
  )
  n98 = Multiply(
   A=VectorParameter() {group=SuperLevel, parameterName=EmissiveTint2, defaultValue=<rgba 0.2175 0.3656 0.435 1>}.RGB
   B=ScalarParameter() {defaultValue=1, group=SuperLevel, parameterName=EmissiveTint2_EmissiveMult}
  )
  n99 = ScalarParameter() {defaultValue=1, group=SuperLevel, parameterName=EmissiveTint2_Strength}
  n100 = Saturate(None=Subtract(A=NamedRerouteUsage(), B=OneMinus(None=@n99)))
  n101 = VectorParameter() {group=SuperLevel, parameterName=EdgeColor, defaultValue=<rgba 0.3586 0.604 0.6094 1>}
  n102 = LinearInterpolate(
   A=@n101.RGB
   B=VectorParameter() {group=SuperLevel, parameterName=EdgeColor2, defaultValue=<rgba 0.3586 0.604 0.6094 1>}.RGB
   Alpha=NamedRerouteUsage()
  )
  n103 = NamedRerouteUsage()
  n104 = Constant() {r=1}
  n105 = Clamp(None=Subtract(
    A=MaterialFunctionCall(In=@n103, Contrast=@n104) {materialFunction=&CheapContrast}.Result
    B=NamedRerouteUsage()
   ) {constB=0.5}) {maxDefault=0.5}
  n106 = Multiply(
   A=StaticSwitchParameter(True=LinearInterpolate(A=@n102, Alpha=@n105), False=@n101.RGB) {parameterName=T3}
   B=ScalarParameter() {defaultValue=0.5, group=SuperLevel, parameterName=Edge_EmissiveMult}
  )
  n107 = Multiply(A=Constant() {r=1}, B=ScalarParameter() {defaultValue=0.1, parameterName=PanningEmissiveTiling2})
  n108 = Divide(
   A=VertexInterpolator(VS=PreSkinnedPosition()).PS
   B=ScalarParameter() {defaultValue=40, group=PanningEmissive, parameterName=Triplanner_ConstantScaleDivisor}
  )
  n109 = Multiply(A=@n107, B=@n108) {constA=2}
  n110 = Frac(None=Multiply(A=Time(), B=VectorParameter() {group=PanningEmissive, parameterName=TriPlanner Speed2, defaultValue=<rgba 0 0 0.15 1>}.RGB))
  n111 = TextureSampleParameter2D(UVs=ComponentMask(None=Add(A=@n109, B=@n110)) {b=true, g=true}) {parameterName=Panning Texture2, samplerType=SAMPLERTYPE_Masks, texture=/Game/Weapons/WeaponSkins/Wraps/Textures/Mash/Mash_Noise}
  n112 = Multiply(A=@n107, B=@n108) {constA=2}
  n113 = TextureSampleParameter2D(UVs=ComponentMask(None=Add(A=@n112, B=@n110)) {b=true, r=true}) {parameterName=Panning Texture2, samplerType=SAMPLERTYPE_Masks, texture=&Mash_Noise}
  n114 = VertexInterpolator(VS=PreSkinnedNormal())
  n115 = ComponentMask(None=@n114.PS) {g=true}
  n116 = ScalarParameter() {defaultValue=1, group=PanningEmissive, parameterName=TriPlanner_AxisBlendStrength}
  n117 = ComponentMask(None=MaterialFunctionCall(In=Abs(None=@n115), Contrast=@n116) {materialFunction=&CheapContrast}.Result) {g=true}
  n118 = Multiply(A=@n107, B=@n108) {constA=2}
  n119 = TextureSampleParameter2D(UVs=ComponentMask(None=Add(A=@n118, B=@n110)) {g=true, r=true}) {parameterName=Panning Texture2, samplerType=SAMPLERTYPE_Masks, texture=&Mash_Noise}
  n120 = MaterialFunctionCall(In=Abs(None=ComponentMask(None=@n114.PS) {b=true}), Contrast=@n116) {materialFunction=&CheapContrast}
  n121 = LinearInterpolate(
   A=LinearInterpolate(A=@n111.R, B=@n113.R, Alpha=Saturate(None=@n117))
   B=@n119.R
   Alpha=Saturate(None=ComponentMask(None=@n120.Result) {b=true})
  )
  n122 = LinearInterpolate(A=@n113.R, B=@n119.R)
  n123 = ScalarParameter() {group=PanningEmissive, parameterName=Triplanner_Z axis Only}
  n124 = Multiply(
   A=NamedRerouteUsage()
   B=LinearInterpolate(
    A=ScalarParameter() {defaultValue=0.1, group=PanningEmissive, parameterName=PanningEmissiveTint2_Min}
    B=ScalarParameter() {defaultValue=3, group=PanningEmissive, parameterName=PanningEmissiveTint2_Max}
    Alpha=LinearInterpolate(A=@n121, B=@n122, Alpha=@n123)
   )
  )
  NamedRerouteDeclaration(None=Add(A=LinearInterpolate(A=@n97, B=@n98, Alpha=@n100), B=Multiply(A=@n106, B=@n124))) {name=Emissive, nodeColor=<rgba 0 0.0156 1 1>}
  n125 = ScalarParameter() {defaultValue=1, parameterName=NormalStrength}
  NamedRerouteDeclaration(None=MaterialFunctionCall(Normal=@n9.Normal, Flatness=OneMinus(None=@n125)) {materialFunction=&FlattenNormal}.Result) {name=Normal_Boosted, nodeColor=<rgba 0 1 0.7656 1>}
  NamedRerouteUsage()
  NamedRerouteUsage()
  ObjectPositionWS()
  n126 = MaterialFunctionCall(Float2=TextureCoordinate()) {materialFunction=&BreakOutFloat2Components}
  n127 = ScalarParameter() {group=PanningEmissive, parameterName=Spiral_U_Offset}
  n128 = ScalarParameter() {group=PanningEmissive, parameterName=Spiral_V_Offset}
  n129 = Multiply(
   A=AppendVector(A=Subtract(A=@n126.R, B=@n127), B=Subtract(A=@n126.G, B=@n128))
   B=ScalarParameter() {defaultValue=1, group=PanningEmissive, parameterName=Spiral_UV_Scale}
  )
  n130 = ComponentMask(None=VectorParameter() {group=PanningEmissive, parameterName=Spiral_UV_Center, defaultValue=<rgba 0.5 0.5 0 1>}.RGB) {g=true, r=true}
  n131 = MaterialFunctionCall!(Vector or UVs=VertexInterpolator(VS=Subtract(A=@n129, B=@n130) {constB=0.5}).PS) {materialFunction=/Game/Athena/SafeZone/S11/MF_VectorToRadialValueFast}
  n132 = ComponentMask(None=@n131.Result) {r=true}
  MaterialExpressionFrac_5 = Frac(None=@n132)
  MaterialExpressionSubtract_6 = Subtract(A=Frac(None=Add(A=@n132) {constB=0.5})) {constB=0.5}
  n133 = Custom!(u0=@MaterialExpressionFrac_5, u1=@MaterialExpressionSubtract_6) {
   code="return ((fwidth(u0) < fwidth(u1)-0.001)? u0:u1);"
   description=FixPerPixelCylidricalUV
   outputType=CMOT_Float1
   inputs [
    {
     inputName=u0
     input {expression=&self:MaterialExpressionFrac_5}
    }
    {
     inputName=u1
     input {expression=&self:MaterialExpressionSubtract_6}
    }
   ]
  }
  n134 = ComponentMask(None=@n133) {r=true}
  n135 = Time()
  n136 = ScalarParameter() {defaultValue=0.5, group=PanningEmissive, parameterName=Spiral Mask1 Speed}
  n137 = Add(A=@n134, B=Frac(None=Multiply(A=@n135, B=@n136)))
  n138 = ComponentMask(None=@n131.Result) {g=true}
  n139 = Multiply(A=@n138, B=ScalarParameter() {defaultValue=4, group=PanningEmissive, parameterName=Spiral Mask1 VTiling}) {constB=4}
  n140 = TextureSampleParameter2D(UVs=AppendVector(A=@n137, B=Add(A=@n137, B=@n139))) {
   group=PanningEmissive
   parameterName=Spiral Mask1
   samplerSource=SSM_Wrap_WorldGroupSettings
   samplerType=SAMPLERTYPE_Masks
   texture=/Game/Global/Textures/Default/Masks_and_Gradients/T_VertGradient_MASK
  }
  n141 = ScalarParameter() {defaultValue=0.5, group=PanningEmissive, parameterName=Spiral Mask2 Speed}
  n142 = Add(A=@n134, B=Frac(None=Multiply(A=@n135, B=@n141) {constB=-1}))
  n143 = Multiply(A=@n138, B=ScalarParameter() {defaultValue=2, group=PanningEmissive, parameterName=Spiral Mask2 VTiling}) {constB=2}
  n144 = TextureSampleParameter2D(UVs=AppendVector(A=@n142, B=Add(A=@n143, B=@n142))) {
   group=PanningEmissive
   parameterName=Spiral Mask2
   samplerSource=SSM_Wrap_WorldGroupSettings
   samplerType=SAMPLERTYPE_Masks
   texture=&T_VertGradient_MASK
  }
  OneMinus(None=Multiply(A=@n140.R, B=@n144.R))
  ScalarParameter() {defaultValue=1, group=SuperLevel, parameterName=EmissiveTint1_EmissiveMult}
  StaticSwitchParameter(True=@n72, False=@n71.RGB) {parameterName=LW_T3_Ronin}
  n145 = ScalarParameter() {defaultValue=1, group=SuperLevel, parameterName=EmissiveTint1_Strength}
  n146 = NamedRerouteUsage()
  Subtract(A=@n92, B=OneMinus(None=LinearInterpolate(A=@n145, Alpha=@n146)))
  TextureSample() {
   samplerType=SAMPLERTYPE_Masks
   texture=/BRCosmetics/Characters/Player/Male/Medium/Bodies/M_MED_Lone_Wolf/Textures/T_M_MED_Lone_Wolf_SL_Masks_T3
  }
 ]
}
